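; Gallery encounter monitor that runs a rescue scenario, auto-hiring a rescued protagonist AI to a player.
; On playlist ready or forced enable, it activates Living World point providers for the protagonist and antagonists and tracks their spawns.
; The second antagonist death swaps music tags from rescuable to rescued, then can cue a cheer and delayed conversion to the killer's pawn.
; A sphere overlap provides fallback conversion, screened by refusal tags and existing hires.
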(event Custom|ForceEnableGalleryAI
  (bind _returnvalue (Class|FortAthenaLivingWorldManager|GetLivingWorldManager))
  (bind _protagonistprovider (Variables|Default|GetProtagonistProvider))
  (Class|FortAthenaLivingWorldPointProviderInterface|EnablePointProvider_DEPRECATED _protagonistprovider)
  (EventDispatchers|BindEventtoOnActorSpawnedResult _protagonistprovider (EventDispatchers|CreateEvent :!event Custom|ProtagonistSpawned))
  (for _array_element (Variables|Default|GetAntagonistProviders)
    (Class|FortAthenaLivingWorldPointProviderInterface|EnablePointProvider_DEPRECATED _array_element)
    (EventDispatchers|BindEventtoOnActorSpawnedResult _array_element (EventDispatchers|CreateEvent :!event Custom|AntagonistSpawned))
    (Utilities|IsValid _returnvalue
      (:Is%20Valid
        (Class|FortAthenaLivingWorldManager|RequestEventGeneration _returnvalue)
        (Variables|Default|SetDeathCount 0)))))

(event EventBeginPlay
  (bind _returnvalue (Class|FortAthenaLivingWorldManager|GetLivingWorldManager))
  (bind _protagonistprovider (Variables|Default|GetProtagonistProvider))
  (Class|FortAsyncActionCurrentPlaylistReady|PlaylistDataReadyAsync
    (:OnReady
      (if (Variables|Default|GetIsEnabled)
        (Class|FortAthenaLivingWorldPointProviderInterface|EnablePointProvider_DEPRECATED _protagonistprovider)
        (EventDispatchers|BindEventtoOnActorSpawnedResult _protagonistprovider (EventDispatchers|CreateEvent :!event Custom|ProtagonistSpawned))
        (for _array_element (Variables|Default|GetAntagonistProviders)
          (Class|FortAthenaLivingWorldPointProviderInterface|EnablePointProvider_DEPRECATED _array_element)
          (EventDispatchers|BindEventtoOnActorSpawnedResult _array_element (EventDispatchers|CreateEvent :!event Custom|AntagonistSpawned))
          (Utilities|IsValid _returnvalue
            (:Is%20Valid
              (Class|FortAthenaLivingWorldManager|RequestEventGeneration _returnvalue)
              (Variables|Default|SetDeathCount 0))))))))

(event Collision|EventActorBeginOverlap
  (params (OtherActor :type Actor%20Object%20Reference)))

(event EventTick
  (params (DeltaSeconds :type Float%20%28single-precision%29)))

(event Custom|ProtagonistSpawned
  (params (SpawnedActor :type Actor%20Object%20Reference) (bResult :type Boolean))
  (bind _asfort_pawn (Utilities|Casting|CastToFortPawn SpawnedActor))
  (bind _output_get (Variables|InternalData|SetProtagonistPawn _asfort_pawn))
  (Ability|GameplayEffect|AddLooseGameplayTags _output_get (Variables|Music|GetRescuable%20Tag) true))

(event Custom|AntagonistSpawned
  (params (SpawnedActor :type Actor%20Object%20Reference) (bResult :type Boolean))
  (bind _asfort_pawn (Utilities|Casting|CastToFortPawn SpawnedActor))
  (Utilities|Array|AddUnique (Variables|InternalData|GetAntagonistsPawns) _asfort_pawn)
  (EventDispatchers|BindEventtoOnDied _asfort_pawn (EventDispatchers|CreateEvent :!event Custom|AntagonistDied)))

(event Custom|AntagonistDied
  (params (DamagedActor :type Actor%20Object%20Reference) (Damage :type Float%20%28single-precision%29) (InstigatedBy :type Controller%20Object%20Reference) (DamageCauser :type Actor%20Object%20Reference) (HitLocation :type Vector) (FHitComponent :type Primitive%20Component%20Object%20Reference) (BoneName :type Name) (Momentum :type Vector))
  (bind _lastvalidconverter (Variables|Default|GetLastValidConverter))
  ;[1 1 1] When Enemies have been Eliminated - Check to Auto Hire the Protagonist
  (bind _result (Math|Integer|IncrementInt (Variables|Default|GetDeathCount))
    (:%20%20
      (if (== _result 2)
        (CallFunction|StopMusic)
        (bind _asfort_pawn (Utilities|Casting|CastToFortPawn (Pawn|GetControlledPawn InstigatedBy))
          (:then
            (if (ScalableFloat|GetValueasBool (Variables|Default|GetPreventAutoHireIfHasConvertedPawn))
              (bind _hasconvertedpawn (Class|BPFLAICommonFunctions|HasConvertedPawn _asfort_pawn))
              (if _hasconvertedpawn
                (Variables|/Script/Engine.PrimitiveComponent|SetbGenerateOverlapEvents true (Variables|Default|GetSphereCollision)) ; Activate Proximity Trigger for next Auto Hire
                (else
                  _))
              (else
                _))
            (Utilities|FlowControl|Delay (ScalableFloat|GetValueatLevel (Variables|Default|GetDelayBeforeCheer)))
            (bind _output_get (Variables|Default|SetLastValidConverter _asfort_pawn))
            (Ability|SendGameplayEventtoActor (Variables|InternalData|GetProtagonistPawn) "(TagName=\"Event.Ability.AIGallery.DoCheer\")" (Utilities|Struct|MakeGameplayEventData "(TagName=\"\")" 0 0 _output_get))
            (Components|DestroyComponent (Variables|Default|GetSphereCollision))
            (Utilities|FlowControl|Sequence
              (:then_0
                (if (and (> Damage 0) (Utilities|IsValid DamageCauser :!class /Script/Engine.KismetSystemLibrary))
                  (Default|CallAIWasRescued)))
              (:then_1
                (Utilities|FlowControl|Delay 5.0)
                (if (not (GameplayTags|HasMatchingGameplayTag (Variables|InternalData|GetProtagonistPawn) (Variables|Default|GetTag_Converted)))
                  (Utilities|IsValid _lastvalidconverter
                    (:Is%20Valid
                      (Convert|TriggerConversion _lastvalidconverter (Variables|InternalData|GetProtagonistPawn))))))))
          (:CastFailed
            _))))))

(event OnComponentBeginOverlap(SphereCollision)
  (params (OverlappedComponent :type Primitive%20Component%20Object%20Reference) (OtherActor :type Actor%20Object%20Reference) (OtherComp :type Primitive%20Component%20Object%20Reference) (OtherBodyIndex :type Integer) (bFromSweep :type Boolean) (SweepResult :type Hit%20Result%20Structure%20%28by%20ref%29))
  (bind _spherecollision (Variables|Default|GetSphereCollision))
  ;[1 1 1] Fall back in case no players were the one that did the last hit on the Stormtroopers.
  (bind _asfort_pawn (Utilities|Casting|CastToFortPawn OtherActor))
  (if (GameplayTags|HasAnyMatchingGameplayTags _asfort_pawn (Variables|Default|GetTags_RefuseConversationIfDeathFailed))
    _
    (elif (ScalableFloat|GetValueasBool (Variables|Default|GetPreventAutoHireIfHasConvertedPawn))
      (bind _hasconvertedpawn (Class|BPFLAICommonFunctions|HasConvertedPawn _asfort_pawn))
      (if (not _hasconvertedpawn)
        (Utilities|FlowControl|DoOnce false
          (:Completed
            (Convert|TriggerConversion _asfort_pawn (Variables|InternalData|GetProtagonistPawn))
            (Variables|/Script/Engine.PrimitiveComponent|SetbGenerateOverlapEvents false _spherecollision)
            (Components|DestroyComponent _spherecollision)
            (Default|CallAIWasRescued))))
      (else
        _))))

(event Custom|Stop%20Music
  ;[1 1 1] Stop music
  (Ability|GameplayEffect|RemoveLooseGameplayTags (Variables|InternalData|GetProtagonistPawn) (Variables|Music|GetRescuable%20Tag) true)
  (Ability|GameplayEffect|AddLooseGameplayTags (Variables|InternalData|GetProtagonistPawn) (Variables|Music|GetRescued%20Tag) true))

(fn GetSmokeStackLocation
  (return (+ (Transformation|GetActorLocation self) (Variables|Default|GetSmokeStackLocation))))

(fn ConstructionScript)
