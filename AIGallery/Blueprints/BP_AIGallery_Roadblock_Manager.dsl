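; Server-authoritative manager that, once playlist data passes tag exclusions, spawns one random or debug-forced AI gallery and trooper roadblocks at the other locations.
; Parallel, length-validated arrays drive data layer activation and point provider enabling.
; A replicated smoke signal location makes clients spawn smoke at the gallery, cleared when the AI is rescued or the nearest antenna is destroyed.
; TA_ForceSpawnLoc tears everything down and respawns at a chosen location.
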(event Custom|TryCustomSpawn
  (bind _galleryspawnlocations (Variables|Default|GetGallerySpawnLocations))
  (bind _returnvalue (Utilities|Array|Length _galleryspawnlocations))
  (bind _self self)
  (bind _debugforcespawnlocation (Variables|InternalData|GetDebugForceSpawnLocation))
  (Networking|SwitchHasAuthority
    (:Authority
      (if (Variables|Default|GetIsEnabled)
        (if (ScalableFloat|GetValueasBool (Variables|Default|GetHF_EnableRoadBlockManager))
          (bind (_gamestate _playlist _playlistcontexttags) (Class|FortAsyncActionCurrentPlaylistReady|PlaylistDataReadyAsync)
            (:OnReady
              (if (GameplayTags|HasAnyTags _playlistcontexttags (Variables|Default|GetPlaylistTagsToExclude))
                _
                (elif (> _returnvalue 0)
                  (if (CallFunction|ValidateArrayLengths)
                    (if (ScalableFloat|GetValueasBool (Variables|Default|GetHF_SpawnGallery))
                      (if (and (> _debugforcespawnlocation 0) (<= _debugforcespawnlocation _returnvalue))
                        (Variables|InternalData|SetSelectedGalleryIndex (- _debugforcespawnlocation 1))
                        (else
                          (bind _output_get (Variables|InternalData|SetSelectedGalleryIndex (Math|Random|RandomInteger _returnvalue)))))
                      (Class|BPAIGalleryMonitor|ForceEnableGalleryAI (Utilities|Array|Get(acopy) _galleryspawnlocations _output_get))
                      (DataLayers|SetDataLayerRuntimeStatebyLabel (WorldSubsystems|GetDataLayerSubsystem) (Utilities|Array|Get(acopy) (Variables|Default|GetGalleryDataLayerNames) (Variables|InternalData|GetSelectedGalleryIndex)) "Activated" false)
                      (Default|BindEventtoAIWasRescued (Utilities|Array|Get(acopy) _galleryspawnlocations _output_get) (EventDispatchers|CreateEvent :!event Custom|AIWasRescued))
                      (Variables|Default|SetSmokeSignalLocation (Class|BPAIGalleryMonitor|GetSmokeStackLocation (Utilities|Array|Get(acopy) _galleryspawnlocations _output_get) :!class /AIGallery/Blueprints/BP_AIGallery_Monitor.BP_AIGallery_Monitor_C))
                      (else
                        _))
                    (for (_array_element _array_index) (Variables|Default|GetTrooperRoadblockProviders)
                      (if (ScalableFloat|GetValueasBool (Variables|Default|GetHF_SpawnGallery))
                        (if (!= _array_index (Variables|InternalData|GetSelectedGalleryIndex))
                          _)
                        (else
                          _))
                      (Utilities|Array|Add (Variables|InternalData|GetFinalTrooperRoadblockProviders) _array_element)
                      (DataLayers|SetDataLayerRuntimeStatebyLabel (WorldSubsystems|GetDataLayerSubsystem) (Utilities|Array|Get(acopy) (Variables|Default|GetRoadblockDataLayerNames) _array_index) "Activated" false))
                    (for _array_element_2 (Variables|InternalData|GetFinalTrooperRoadblockProviders)
                      (for _array_element_3 (Utilities|Struct|BreakSTRUCTAIGalleryProviders _array_element_2)
                        (Class|FortAthenaLivingWorldPointProviderInterface|EnablePointProvider_DEPRECATED _array_element_3)))
                    (Utilities|FlowControl|Delay 5.0)
                    (CallFunction|MonitorClosestAntennatoSmokeStack)
                    (else
                      (Fortnite|FortShippingLog _self "AIGallery Manager Count does not match Dark Troopers and Road Block Count. Aborting trying to spawn Gallery AI." true)))
                  (else
                    (Fortnite|FortShippingLog _self "Trying to Spawn AIGallery and Roadblocks, but no Gallery Managers were configured." true)))))))))))

(event EventBeginPlay
  (bind _galleryspawnlocations (Variables|Default|GetGallerySpawnLocations))
  (bind _returnvalue (Utilities|Array|Length _galleryspawnlocations))
  (bind _self self)
  (bind _debugforcespawnlocation (Variables|InternalData|GetDebugForceSpawnLocation))
  (Networking|SwitchHasAuthority
    (:Authority
      (if (Variables|Default|GetIsEnabled)
        (if (ScalableFloat|GetValueasBool (Variables|Default|GetHF_EnableRoadBlockManager))
          (bind (_gamestate _playlist _playlistcontexttags) (Class|FortAsyncActionCurrentPlaylistReady|PlaylistDataReadyAsync)
            (:OnReady
              (if (GameplayTags|HasAnyTags _playlistcontexttags (Variables|Default|GetPlaylistTagsToExclude))
                _
                (elif (> _returnvalue 0)
                  (if (CallFunction|ValidateArrayLengths)
                    (if (ScalableFloat|GetValueasBool (Variables|Default|GetHF_SpawnGallery))
                      (if (and (> _debugforcespawnlocation 0) (<= _debugforcespawnlocation _returnvalue))
                        (Variables|InternalData|SetSelectedGalleryIndex (- _debugforcespawnlocation 1))
                        (else
                          (bind _output_get (Variables|InternalData|SetSelectedGalleryIndex (Math|Random|RandomInteger _returnvalue)))))
                      (Class|BPAIGalleryMonitor|ForceEnableGalleryAI (Utilities|Array|Get(acopy) _galleryspawnlocations _output_get))
                      (DataLayers|SetDataLayerRuntimeStatebyLabel (WorldSubsystems|GetDataLayerSubsystem) (Utilities|Array|Get(acopy) (Variables|Default|GetGalleryDataLayerNames) (Variables|InternalData|GetSelectedGalleryIndex)) "Activated" false)
                      (Default|BindEventtoAIWasRescued (Utilities|Array|Get(acopy) _galleryspawnlocations _output_get) (EventDispatchers|CreateEvent :!event Custom|AIWasRescued))
                      (Variables|Default|SetSmokeSignalLocation (Class|BPAIGalleryMonitor|GetSmokeStackLocation (Utilities|Array|Get(acopy) _galleryspawnlocations _output_get) :!class /AIGallery/Blueprints/BP_AIGallery_Monitor.BP_AIGallery_Monitor_C))
                      (else
                        _))
                    (for (_array_element _array_index) (Variables|Default|GetTrooperRoadblockProviders)
                      (if (ScalableFloat|GetValueasBool (Variables|Default|GetHF_SpawnGallery))
                        (if (!= _array_index (Variables|InternalData|GetSelectedGalleryIndex))
                          _)
                        (else
                          _))
                      (Utilities|Array|Add (Variables|InternalData|GetFinalTrooperRoadblockProviders) _array_element)
                      (DataLayers|SetDataLayerRuntimeStatebyLabel (WorldSubsystems|GetDataLayerSubsystem) (Utilities|Array|Get(acopy) (Variables|Default|GetRoadblockDataLayerNames) _array_index) "Activated" false))
                    (for _array_element_2 (Variables|InternalData|GetFinalTrooperRoadblockProviders)
                      (for _array_element_3 (Utilities|Struct|BreakSTRUCTAIGalleryProviders _array_element_2)
                        (Class|FortAthenaLivingWorldPointProviderInterface|EnablePointProvider_DEPRECATED _array_element_3)))
                    (Utilities|FlowControl|Delay 5.0)
                    (CallFunction|MonitorClosestAntennatoSmokeStack)
                    (else
                      (Fortnite|FortShippingLog _self "AIGallery Manager Count does not match Dark Troopers and Road Block Count. Aborting trying to spawn Gallery AI." true)))
                  (else
                    (Fortnite|FortShippingLog _self "Trying to Spawn AIGallery and Roadblocks, but no Gallery Managers were configured." true)))))))))))

(event Custom|AIWasRescued
  ;[1 1 1] Disable SMoke Stack
  (Variables|Default|SetSmokeSignalLocation "0, 0, 0"))

(event Custom|TA_ForceSpawnLoc
  (params (Location :type Integer))
  (bind _returnvalue (WorldSubsystems|GetDataLayerSubsystem))
  ;[1 1 1] CHeats to choose which Location to Spawn
  (if (and (> Location 0) (<= Location (Utilities|Array|Length (Variables|Default|GetGallerySpawnLocations))))
    (Variables|InternalData|SetDebugForceSpawnLocation Location)
    (CallFunction|DisableSmokeSignals)
    (CallFunction|DisableAllGallerySpawnLocations)
    (for (_array_element _array_index) (Variables|Default|GetTrooperRoadblockProviders)
      (for _array_element_2 (Utilities|Struct|BreakSTRUCTAIGalleryProviders _array_element)
        (Class|FortAthenaLivingWorldPointProviderInterface|DisablePointProvider _array_element_2))
      (DataLayers|SetDataLayerRuntimeStatebyLabel _returnvalue (Utilities|Array|Get(aref) (Variables|Default|GetGalleryDataLayerNames) _array_index) "Unloaded" false)
      (DataLayers|SetDataLayerRuntimeStatebyLabel _returnvalue (Utilities|Array|Get(aref) (Variables|Default|GetRoadblockDataLayerNames) _array_index) "Unloaded" false))
    (Utilities|Array|Clear (Variables|InternalData|GetFinalTrooperRoadblockProviders))
    (Utilities|FlowControl|Delay 2.0)
    (CallFunction|TryCustomSpawn)))

(event Custom|DisableSmokeSignals
  ;[1 1 1] CHeats to choose which Location to Spawn
  (bind _smokesignal (Variables|InternalData|GetSmokesignal)
    (:then
      (Class|BPTravelAgentSmokesignal|DisableSmoke _smokesignal))
    (:else)))

(event Custom|MonitorClosestAntennaToSmokeStack
  (bind _smokesignallocation (Variables|Default|GetSmokeSignalLocation))
  (bind _smokestackantennaactor (Variables|InternalData|GetSmokeStackAntennaActor))
  ;[1 1 1] Track Antenna Damage
  (bind _outactors (Actor|GetAllActorsOfClass "/Game/Athena/Artemis/Environments/BuildingActors/Realm/Artemis_Realm_Antenna_B.Artemis_Realm_Antenna_B_C"))
  (for _array_element _outactors
    (if (< (Math|Vector|Distance(Vector) (Transformation|GetActorLocation _array_element) _smokesignallocation) (Variables|InternalData|GetSmokeStackDistance))
      (Variables|InternalData|SetSmokeStackDistance (Math|Vector|Distance(Vector) (Transformation|GetActorLocation _array_element) _smokesignallocation))
      (Variables|InternalData|SetSmokeStackAntennaActor _array_element)))
  (Utilities|IsValid _smokestackantennaactor
    (:Is%20Valid
      (bind _asbuilding_prop (Utilities|Casting|CastToBuildingProp _smokestackantennaactor))
      (EventDispatchers|BindEventtoOnDamaged _asbuilding_prop (EventDispatchers|CreateEvent :!event Custom|AntennaTookDamage)))))

(event Custom|AntennaTookDamage
  (params (DamagedActor :type Actor%20Object%20Reference) (Damage :type Float%20%28single-precision%29) (InstigatedBy :type Controller%20Object%20Reference) (DamageCauser :type Actor%20Object%20Reference) (HitLocation :type Vector) (FHitComponent :type Primitive%20Component%20Object%20Reference) (BoneName :type Name) (Momentum :type Vector))
  ;[1 1 1] Track Antenna Damage
  (bind _asbuilding_prop (Utilities|Casting|CastToBuildingProp DamagedActor))
  (if (<= (Health|GetHealth _asbuilding_prop :!class /Script/FortniteGame.BuildingActor) 0)
    (Variables|Default|SetSmokeSignalLocation "0, 0, 0")))

(fn DisableAllGallerySpawnLocations
  (for _array_element (Variables|Default|GetGallerySpawnLocations)
    (Variables|/AIGallery/Blueprints/BP_AIGallery_Monitor.BP_AIGallery_Monitor_C|SetIsEnabled false _array_element)
    (Class|FortAthenaPatrolPathPointProvider|DisablePointProvider (Variables|/AIGallery/Blueprints/BP_AIGallery_Monitor.BP_AIGallery_Monitor_C|GetProtagonistProvider _array_element))
    (for _array_element_2 (Variables|/AIGallery/Blueprints/BP_AIGallery_Monitor.BP_AIGallery_Monitor_C|GetAntagonistProviders _array_element)
      (Class|FortAthenaPatrolPathPointProvider|DisablePointProvider _array_element_2))))

(fn OnRep_SmokeSignalLocation
  (Networking|SwitchHasAuthority
    (:Remote
      (if (Math|Vector|VectorIsNearlyZero (Variables|Default|GetSmokeSignalLocation))
        (bind _smokesignal (Variables|InternalData|GetSmokesignal)
          (:then
            (Class|BPTravelAgentSmokesignal|DisableSmoke _smokesignal))
          (:else))
        (else
          (bind _returnvalue (Game|SpawnActorfromClass (Variables|Default|GetSmokeStack) (Math|Transform|MakeTransform (Variables|Default|GetSmokeSignalLocation)) "AlwaysSpawn" "MultiplyWithRoot" self 0))
          (bind _asbp_travel_agent_smokesignal (Utilities|Casting|CastToBP_TravelAgent_Smokesignal _returnvalue))
          (bind _output_get (Variables|InternalData|SetSmokesignal _asbp_travel_agent_smokesignal))
          (Rendering|SetActorHiddenInGame _output_get false)
          (Class|BPTravelAgentSmokesignal|EnableNearParticles _output_get))))))

(fn ConstructionScript)

(fn ValidateArrayLengths
  (LocalVariables|SetGallerySpawnLocationsArrayLength (Utilities|Array|Length (Variables|Default|GetGallerySpawnLocations)))
  (return (Math|Boolean|ANDBoolean (== (LocalVariables|GetGallerySpawnLocationsArrayLength) (Utilities|Array|Length (Variables|Default|GetTrooperRoadblockProviders))) (== (LocalVariables|GetGallerySpawnLocationsArrayLength) (Utilities|Array|Length (Variables|Default|GetGalleryDataLayerNames))) (== (LocalVariables|GetGallerySpawnLocationsArrayLength) (Utilities|Array|Length (Variables|Default|GetRoadblockDataLayerNames))))))
