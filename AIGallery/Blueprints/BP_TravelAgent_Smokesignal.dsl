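; Smoke signal actor for the travel agent that drives a Niagara effect component through two custom events.
; EnableNearParticles waits one second before activating the smoke with a reset, noted as a fix for activating after being unhidden in game.
; DisableSmoke deactivates the effect, waits for the LifeTimeAfterDeath scalable float duration, then destroys the actor.
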(event Custom|EnableNearParticles
  ;[0 0 0] fix for activating, one frame delay after being set unhidden in game
  (Utilities|FlowControl|Delay 1.0)
  (Components|Activation|Activate (Variables|Default|GetNS_TravelAgent_SmokeSignal) true))

(event Custom|DisableSmoke
  (Components|Activation|SetActive (Variables|Default|GetNS_TravelAgent_SmokeSignal) false false)
  (Utilities|FlowControl|Delay (ScalableFloat|GetValueatLevel (Variables|Default|GetLifeTimeAfterDeath)))
  (Actor|DestroyActor))

(fn ConstructionScript)
